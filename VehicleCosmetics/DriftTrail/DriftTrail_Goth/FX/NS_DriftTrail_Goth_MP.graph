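# Niagara vehicle cosmetic for the Goth drift trail, drawing dark monochrome trails and smoke-like particles from two wheels, apparently the rear pair.
# Two ribbons, each drawn by three renderers, animate width, color and oscillating twist under curl noise.
# Mesh and sprite particles spawn at wheel bones and swirl with vortex, noise and drag.
# Wheel-wrap meshes follow those bones; they and the ribbons use tire-FX masking, and painted-variant overrides are disabled throughout.
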NiagaraSystem /VehicleCosmetics/DriftTrail/DriftTrail_Goth/FX/NS_DriftTrail_Goth_MP
dependencies {
 usedDataInterfaces [
  &NiagaraDataInterfaceCurve
  &NiagaraDataInterfaceColorCurve
  &NiagaraDataInterfaceVectorCurve
  &NiagaraDataInterfaceSkeletalMesh
 ]
 usedDynamicInputs []
 usedModules [
  &self:SystemSpawnScript.SystemScriptSource.SystemScriptGraph.NiagaraNodeAssignment_0.SetVariables_#1
  &FN_UnpackVehicleFlags
  /VehicleCosmetics/FX/Niagara/Modules/VC_SetVehicleSpeedAttribs
  &SystemState
 ]
}
emitters {
 NE_VC_DriftTrail_Ribbon_Parent003 {
  emitterUpdateScript [
   &EmitterState()
   &VC_RateQualityHandler(Use Custom Rate?=true)
   &VC_SpawnRibbon_v1(Number of Ribbons=2, Spawn Time=-1, SpawnRate=$Emitter.Spawn Rate)
  ]
  particleSpawnScript [
   &VC_PaintedVariantsOverride() {enabled=false}
   /SetVariables(Particles.BlackWidth=130, Particles.GreyWidth=80, Particles.ThirdWidth=50) {isSetParameters=true}
   &VC_InitializeRibbon_v1(
    Custom Offset=true
    Enable Skeletal Position=true
    Facing Axis:ENiagaraOrientationAxis=::ZAxis
    Lifetime=&MakeCustomFloatFromBool(Bool=true, True Float=0.5, False Float=0.8)
    Local Vector Offset=&AddVectorToPosition(
     Position=&MakeCustomVectorFromBool(Bool=$User.bIsFrontendPreview, True Vector=<80 0 -20>, False Vector=<100 0 -20>)
     Vector=$User.offset
    )
    Offset Amount=1
    Ribbon Width=1
   )
   &AddVelocity(Velocity=<-2500 0 0>) {module=AddVelocity001}
   /SetVariables() {isSetParameters=true}
  ]
  particleUpdateScript [
   &FN_DisableTireFX(Mask Index Attribute:NiagaraInt32=$Particles.SpawnGroup, Mask Index FL=2, Mask Index FR=3, Mask Index RL=0, Mask Index RR=1)
   &ParticleState()
   &InheritSourceMovement(
    Applied Position Scale=&VectorFromFloat(Value=&FloatFromCurve(FloatCurve=DataInterface() {propertyValues {Curve {keys [
          {interpMode=RCIM_Cubic, tangentMode=RCTM_Break, time=0, value=1}
          {arriveTangent=-2.9025, interpMode=RCIM_Cubic, tangentMode=RCTM_Break, time=1, value=0}
         ]}}}))
    ApplyForce=false
    ApplyPosition=true
    ApplyPositionFalloff=false
   )
   /SetVariables(
    Particles.BlackWidth:NiagaraFloat=&Multiply_Float(
     A=$Particles.Initial.BlackWidth
     B=&FloatFromCurve(FloatCurve=DataInterface() {propertyValues {Curve {
         defaultValue=1
         keys [
          {arriveTangent=1.7057, interpMode=RCIM_Cubic, leaveTangent=1.7057, tangentMode=RCTM_User, time=0, value=0.5}
          {arriveTangent=0.6815, interpMode=RCIM_Cubic, leaveTangent=0.6815, tangentMode=RCTM_User, time=0.2438, value=0.6784}
          {arriveTangent=0.0224, interpMode=RCIM_Cubic, leaveTangent=0.0224, tangentMode=RCTM_User, time=1, value=0.8}
         ]
        }}})
    )
    Particles.GreyWidth:NiagaraFloat=&Multiply_Float(
     A=$Particles.Initial.GreyWidth
     B=&FloatFromCurve(FloatCurve=DataInterface() {propertyValues {Curve {
         defaultValue=0.5
         keys [
          {arriveTangent=0.8176, interpMode=RCIM_Cubic, leaveTangent=0.8176, tangentMode=RCTM_User, time=0, value=0.5}
          {
           arriveTangent=0.6167
           arriveTangentWeight=0.1261
           interpMode=RCIM_Cubic
           leaveTangent=0.6167
           tangentMode=RCTM_User
           time=0.2593
           value=0.6599
          }
          {arriveTangent=-0.01, arriveTangentWeight=0.2072, interpMode=RCIM_Cubic, leaveTangent=-0.01, tangentMode=RCTM_User, time=1, value=0.8}
         ]
        }}})
    )
    Particles.R1 Color:LinearColor=&MultiplyLinearColor(
     Linear Color=<rgba 0.005 0.005 0.005 3>
     Scale Factor=&Vector4FromCurve(Vector4Curve=DataInterface() {propertyValues {
        AlphaCurve {
         defaultValue=1
         keys [
          {arriveTangent=-1.2222, interpMode=RCIM_Cubic, time=0.1, value=0}
          {interpMode=RCIM_Cubic, tangentMode=RCTM_User, time=0.2, value=1}
          {tangentMode=RCTM_User, time=0.9295, value=-0.1}
         ]
        }
        BlueCurve {keys [<at 0 1>,<at 0.6118 1>,<at 1 0>]}
        GreenCurve {keys [<at 0 1>,<at 0.6118 1>,<at 1 0>]}
        RedCurve {keys [<at 0 1>,<at 0.6118 1>,<at 1 0>]}
       }})
    )
    Particles.R2 Color:LinearColor=&MultiplyLinearColor(
     Linear Color=<rgba 0.3 0.3 0.3 4>
     Scale Factor=&Vector4FromCurve(Vector4Curve=DataInterface() {propertyValues {
        AlphaCurve {
         defaultValue=1
         keys [
          <at 0.1 0 RCIM_Cubic>
          {interpMode=RCIM_Cubic, tangentMode=RCTM_User, time=0.2, value=1}
          {arriveTangent=-0.0753, interpMode=RCIM_Cubic, leaveTangent=-0.0753, tangentMode=RCTM_User, time=0.85, value=0}
          {tangentMode=RCTM_User, time=1, value=0}
         ]
        }
        BlueCurve {keys [<at 0 1>,<at 1 1>]}
        GreenCurve {keys [<at 0 1>,<at 1 1>]}
        RedCurve {keys [<at 0 1>,<at 1 1>]}
       }})
    )
    Particles.R3 Color:LinearColor=&MultiplyLinearColor(
     Linear Color=<rgba 0.01 0.01 0.01 3>
     Scale Factor=&Vector4FromCurve(Vector4Curve=DataInterface() {propertyValues {
        AlphaCurve {
         defaultValue=1
         keys [
          {arriveTangent=-1.2941, interpMode=RCIM_Cubic, time=0.1, value=0}
          {interpMode=RCIM_Cubic, tangentMode=RCTM_User, time=0.2, value=1}
          {tangentMode=RCTM_User, time=0.9393, value=-0.1}
         ]
        }
        BlueCurve {keys [<at 0 1>,<at 1 0>]}
        GreenCurve {keys [<at 0 1>,<at 1 0>]}
        RedCurve {keys [<at 0 1>,<at 1 0>]}
       }})
    )
   ) {isSetParameters=true}
   &CurlNoiseForce(Bias Noise=false, Noise Strength=50, Pan Noise=false)
   &ScaleVelocity(
    Coordinate Space:ENiagaraCoordinateSpace=::Simulation
    Velocity Scale=&VectorFromCurve(VectorCurve=DataInterface() {propertyValues {XCurve {keys [<at 0.4151 1>,<at 0.5651 1>,<at 0.7151 1>]}, YCurve {keys [<at 0 0>,<at 1 0>]}, ZCurve {keys [<at 0 0>,<at 1 0>]}}})
   ) {enabled=false}
   &SolveForcesAndVelocity(Clamp Velocity=false, Limit Acceleration=false)
   &VC_UpdateRibbon_v1(Enable Skeletal Position=true)
   &DynamicMaterialParameters(
    Index 3 Param 1=&FloatFromCurve(FloatCurve=DataInterface() {propertyValues {Curve {keys [
         {interpMode=RCIM_Cubic, leaveTangent=10.0547, tangentMode=RCTM_Break, time=0, value=-50}
         {arriveTangent=144.5382, interpMode=RCIM_Cubic, leaveTangent=144.5382, tangentMode=RCTM_User, time=0.25, value=-10}
         {arriveTangent=1.6276, interpMode=RCIM_Cubic, tangentMode=RCTM_Break, time=0.4967, value=5}
         <at 1.0024 0 RCIM_Cubic>
        ]}}})
    Module.Index0Float0Write=true
    Module.Index0Float1Write=true
    Module.Index0Float2Write=true
    Module.Index0Float3Write=true
    Module.Index3Float0Write=true
    Module.Index3Float1Write=true
    Module.Index3Float2Write=true
    Module.Index3Float3Write=true
    Write Parameter Index 3=::NewEnumerator1
   )
   /SetVariables(
    Particles.R1Twist:NiagaraFloat=&Add_Float(
     A=$Particles.Initial.R1Twist
     B=&FloatFromCurve(
      FloatCurve=DataInterface() {propertyValues {Curve {
         defaultValue=1
         keys [
          {interpMode=RCIM_Cubic, tangentMode=RCTM_User, time=0, value=-0.15}
          {interpMode=RCIM_Cubic, tangentMode=RCTM_User, time=0.25, value=0.15}
          <at 0.5 -0.15 RCIM_Cubic>
          <at 0.75 0.15 RCIM_Cubic>
          <at 1 -0.15 RCIM_Cubic>
         ]
        }}}
      CurveIndex=$System.NormalizedLoopAge
     )
    )
    Particles.R2Twist:NiagaraFloat=&Add_Float(
     A=$Particles.Initial.R2Twist
     B=&FloatFromCurve(
      FloatCurve=DataInterface() {propertyValues {Curve {
         defaultValue=0.15
         keys [
          {interpMode=RCIM_Cubic, tangentMode=RCTM_User, time=0, value=0.15}
          {arriveTangentWeight=0.0833, interpMode=RCIM_Cubic, tangentMode=RCTM_User, time=0.25, value=-0.15}
          <at 0.5 0.15 RCIM_Cubic>
          <at 0.75 -0.15 RCIM_Cubic>
          <at 1 0.15 RCIM_Cubic>
         ]
        }}}
      CurveIndex=$System.NormalizedLoopAge
     )
    )
   ) {isSetParameters=true}
  ]
  rendererClasses [&NiagaraRibbonRendererProperties]
  renderers [
   {rendererClass=&NiagaraRibbonRendererProperties}
   {rendererClass=&NiagaraRibbonRendererProperties, rendererIndex=1}
   {rendererClass=&NiagaraRibbonRendererProperties, rendererIndex=2}
  ]
 }
 NE_VC_DriftTrail_Sprite_Parent001 {
  emitterUpdateScript [
   &EmitterState()
   &SpawnRate(SpawnRate=50, Use Spawn Probability=false)
  ]
  particleSpawnScript [
   &VC_PaintedVariantsOverride(DefaultColor=<rgba 0.015 0.015 0.015 1>) {enabled=false}
   &InitializeParticle(
    Color=<rgba 0.015 0.015 0.015 1>
    Lifetime Max=1
    Lifetime Min=1.1
    Lifetime Mode=::NewEnumerator1
    Mesh Scale Mode=::NewEnumerator3
    Mesh Uniform Scale=&MakeCustomFloatFromBool(Bool=$User.bIsFrontendPreview, True Float=1, False Float=&RandomRangeFloat(Minimum=1.2, Maximum=2.2))
    Sprite Rotation Mode=::NewEnumerator2
    Sprite Size Mode=::NewEnumerator3
    UsePositionOffset=false
    Write Material Random=true
   )
   &PartitionParticles()
   &SkeletalMeshLocation(
    Bone / Socket Index=&Add_Integer(A=$Particles.PartitionParticles.Partition, B=4)
    Bone / Socket Sampling Mode=::NewEnumerator5
    Interpolate Mesh Sampling=false
    MaskSpawnGroup=false
    Mesh Sampling Type=::NewEnumerator6
    Sampled Position Offset=&AddVectorToPosition(Position=<-50 0 10>, Vector=$User.offset)
    Skeletal Mesh=$System.SkeletalMesh
   )
   &AddVelocity(Velocity=&RandomRangeVector(Minimum=<-750 0 35>, Maximum=<-850 0 100>))
   &ShapeLocation(MaskSpawnGroup=true, Offset=<0 0 -30>, Offset Mode=::NewEnumerator0, Override Seed=false, Sphere Radius=20, UseOwnerScale=true)
   &InitialMeshOrientation(Mesh Axis To Orient=<0 0 0>, Mesh Orientation Mode=::NewEnumerator2, Rotation=<0 0 1>, Use Rotation Vector=true)
  ]
  particleUpdateScript [
   &InheritSourceMovement(
    Applied Position Scale=&VectorFromFloat(Value=&FloatFromCurve(FloatCurve=DataInterface() {propertyValues {Curve {keys [
          {interpMode=RCIM_Cubic, tangentMode=RCTM_Break, time=0, value=1}
          {arriveTangent=-2.9025, interpMode=RCIM_Cubic, tangentMode=RCTM_Break, time=1, value=0}
         ]}}}))
    ApplyForce=false
    ApplyPosition=true
    ApplyPositionFalloff=false
   )
   &ParticleState()
   &ScaleMeshSize(Scale Factor=&VectorFromFloat(Value=&FloatFromCurve(
      FloatCurve=DataInterface() {propertyValues {Curve {defaultValue=1, keys [<at 0 0.2 RCIM_Cubic>,<at 0.15 1 RCIM_Cubic>]}}}
      Scale Curve=&RandomRangeFloat(Minimum=1, Maximum=1.8)
     )))
   &CurlNoiseForce(Bias Noise=false, Noise Frequency=1000, Noise Strength=100, Pan Noise=false)
   &Drag(Drag=0.8, Use Linear Drag=true, Use Rotational Drag=true)
   &VortexForce(
    UseOriginOffset=false
    Vortex Axis=&RandomRangeVector(Minimum=<100 0 0>, Maximum=<125 0 0>)
    Vortex Origin:NiagaraPosition=$Particles.Initial.Position
   )
   &SolveForcesAndVelocity(Clamp Velocity=false, Limit Acceleration=false)
   &DynamicMaterialParameters(
    Index 0 Param 1=&RandomRangeFloat(Minimum=1.15, Maximum=0.7)
    Index 0 Param 2=-1
    Index 0 Param 3=1
    Index 0 Param 4=-1
    Module.Index0Float0Write=true
    Module.Index0Float1Write=true
    Module.Index0Float2Write=true
    Module.Index0Float3Write=true
   )
   &ScaleColor(
    Scale Alpha=&RandomRangeFloat(
     Minimum=&FloatFromCurve(FloatCurve=DataInterface() {propertyValues {Curve {keys [<at 0 1 RCIM_Cubic>,<at 1 0 RCIM_Cubic>]}}}, Scale Curve=1.5)
     Maximum=&FloatFromCurve(FloatCurve=DataInterface() {propertyValues {Curve {keys [<at 0 1 RCIM_Cubic>,<at 1 0 RCIM_Cubic>]}}}, Scale Curve=2)
    )
    ScaleA=true
    ScaleRGB=true
   )
  ]
  rendererClasses [&NiagaraMeshRendererProperties, &NiagaraSpriteRendererProperties]
  renderers [
   {rendererClass=&NiagaraMeshRendererProperties, rendererIndex=0}
   {rendererClass=&NiagaraSpriteRendererProperties, rendererIndex=1}
  ]
 }
 NE_VC_DriftTrail_WheelWrap_Parent {
  emitterUpdateScript [
   &EmitterState(Life Cycle Mode=::NewEnumerator1, Loop Behavior=::NewEnumerator1, Loop Duration Mode=::NewEnumerator1, UseLoopDelay=false)
   &SpawnBurst_Instantaneous(Spawn Count=2, Use Spawn Probability=false, UseLoopCountLimit=false)
  ]
  particleSpawnScript [
   &VC_PaintedVariantsOverride() {enabled=false}
   &InitializeParticle(
    Color Mode=::NewEnumerator0
    Mesh Scale Mode=::NewEnumerator4
    Sprite Rotation Mode=::NewEnumerator2
    UsePositionOffset=false
    Write Material Random=true
   )
   /SetVariables(Particles.R1:LinearColor=<rgba 0.2 0.2 0.2 0.1>, Particles.R2:LinearColor=<rgba 0.001 0.001 0.001 0.05>) {isSetParameters=true}
   &PartitionParticles()
   &SkeletalMeshLocation(
    Bone / Socket Index=&Add_Integer(A=$Particles.PartitionParticles.Partition, B=4)
    Bone / Socket Sampling Mode=::NewEnumerator5
    Interpolate Mesh Sampling=false
    MaskSpawnGroup=false
    Mesh Sampling Type=::NewEnumerator6
    Skeletal Mesh=$System.SkeletalMesh
   )
  ]
  particleUpdateScript [
   &FN_DisableTireFX(
    Mask Index Attribute:NiagaraInt32=$Particles.PartitionParticles.Partition
    Mask Index FL=2
    Mask Index FR=3
    Mask Index RL=0
    Mask Index RR=1
   )
   &VC_WheelWrapScale(Wheel Index Attribute:NiagaraInt32=&Add_Integer(A=$Particles.PartitionParticles.Partition, B=2))
   &ParticleState(Kill Particles When Lifetime Has Elapsed=false)
   &UpdateMeshOrientation(Rotation Coordinate Space=::NewEnumerator2, Rotation Vector=<0 1 0>) {enabled=false}
   &SkeletalMeshLocation(
    Bone / Socket Index=&Add_Integer(A=$Particles.PartitionParticles.Partition, B=4)
    Bone / Socket Sampling Mode=::NewEnumerator5
    MaskSpawnGroup=false
    Mesh Sampling Type=::NewEnumerator6
    Skeletal Mesh=$System.SkeletalMesh
   ) {module=SkeletalMeshLocation001}
   &SolveForcesAndVelocity(Clamp Velocity=false, Limit Acceleration=false)
  ]
  rendererClasses [&NiagaraMeshRendererProperties]
  renderers [
   {rendererClass=&NiagaraMeshRendererProperties, rendererIndex=0}
   {rendererClass=&NiagaraMeshRendererProperties, rendererIndex=1}
   {rendererClass=&NiagaraMeshRendererProperties, rendererIndex=2}
  ]
 }
}
summary {
 systemName=NS_DriftTrail_Goth_MP
 emitters [
  {emitterName=NE_VC_DriftTrail_WheelWrap_Parent, rendererClasses [&NiagaraMeshRendererProperties]}
  {emitterName=NE_VC_DriftTrail_Ribbon_Parent003, rendererClasses [&NiagaraRibbonRendererProperties]}
  {
   emitterName=NE_VC_DriftTrail_Sprite_Parent001
   rendererClasses [&NiagaraMeshRendererProperties, &NiagaraSpriteRendererProperties]
  }
 ]
 userVariables [
  User.bIsFrontendPreview {type=NiagaraBool, defaultValue=true}
  User.bIsPainted {type=NiagaraBool}
  User.ColorVariantID {type=NiagaraInt32}
  User.offset {type=Vector3f, description="offset from wheels(set through cosmetic template)", defaultValue {struct=Vector3f, value=<0 0 0>}}
  User.PaintedColor {type=LinearColor, description="", defaultValue {struct=LinearColor, value=<rgba 1 1 1 1>}}
  User.PaintedScalar {type=NiagaraFloat, defaultValue=1, description=""}
  User.VehicleFlags {type=NiagaraInt32, defaultValue=806289408}
  User.Wheel_Radius_Array {
   type=NiagaraDataInterfaceArrayFloat
   defaultValue {
    struct=NiagaraExt_VariableValue_DataInterface
    value {
     dataInterface=&self:NiagaraDataInterfaceArrayFloat_1
     dataInterfaceClass=&NiagaraDataInterfaceArrayFloat
    }
   }
  }
  User.Wheel_Width_Array {
   type=NiagaraDataInterfaceArrayFloat
   defaultValue {
    struct=NiagaraExt_VariableValue_DataInterface
    value {
     dataInterface=&self:NiagaraDataInterfaceArrayFloat_0
     dataInterfaceClass=&NiagaraDataInterfaceArrayFloat
    }
   }
  }
  User.WheelTouchingArray {
   type=NiagaraDataInterfaceArrayFloat
   defaultValue {
    struct=NiagaraExt_VariableValue_DataInterface
    value {
     dataInterface=&self:NiagaraDataInterfaceArrayFloat_2
     dataInterfaceClass=&NiagaraDataInterfaceArrayFloat
    }
   }
  }
 ]
}
